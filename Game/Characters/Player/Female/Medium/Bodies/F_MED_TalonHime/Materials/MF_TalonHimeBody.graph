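# Material function for the F_MED_TalonHime player body that layers smoke, fresnel and optional eye effects onto incoming material attributes, overriding BaseColor and EmissiveColor.
# Eye emission blends mask-driven colors with rotating, time-animated noise, behind a static switch.
# The backpack's MF_SmokeBeneath builds smoke from a TechArt mask input and emissive.
# Channel-masked, invertible fresnel supports flattened normals and UV vertical fade; a normal-based top gradient is also computed.
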MaterialFunction /Game/Characters/Player/Female/Medium/Bodies/F_MED_TalonHime/Materials/MF_TalonHimeBody
graph {nodes [
  n1 = GetMaterialAttributes(None=FunctionInput() {inputName=InputMaterial, inputType=FunctionInput_MaterialAttributes}) {attributeGetTypes [#1, #2, #3]}
  n2 = StaticComponentMaskParameter(None=NamedRerouteUsage()) {defaultG=true, group=AdditonalTechArt, parameterName=SmokeChannel}
  n3 = OneMinus(None=Saturate(None=ScalarParameter() {group=AdditonalTechArt, parameterName=OriginalBaseColorIntensity}))
  n4 = Multiply(A=@n1.BaseColor, B=OneMinus(None=Multiply(A=@n2, B=@n3)))
  n5 = LinearInterpolate(A=@n4, B=NamedRerouteUsage(), Alpha=Step(X=NamedRerouteUsage()) {constY=0.1}) {constB=0}
  FunctionOutput(None=SetMaterialAttributes!(
    MaterialAttributes=@n1.MaterialAttributes
    EmissiveColor=NamedRerouteUsage()
    BaseColor=StaticSwitchParameter(True=@n5, False=@n4) {group=AdditonalTechArt, parameterName=UseEyeDistortEmission}
   ) {attributeSetTypes [#1, #2]})
  n6 = TextureSample() {samplerType=SAMPLERTYPE_LinearColor, texture=/Game/Characters/Player/Female/Medium/Bodies/F_MED_TalonHime/Textures/T_FacialMask}
  n7 = VectorParameter() {group=AdditonalTechArt_Eye, parameterName=EyeAroundColor, defaultValue=<rgba 0.5552 0.0204 0 1>}
  n8 = Multiply(
   A=StaticComponentMaskParameter(None=@n6.RGB) {defaultG=true, group=AdditonalTechArt_Eye, parameterName=EyeEmissiveChannel}
   B=Multiply(A=NamedRerouteDeclaration(None=@n7.RGB) {name=AroundColor, nodeColor=<rgba 0.5938 0 1 1>}) {constB=3}
  )
  n9 = Rotator(
   Coordinate=Multiply(A=TextureCoordinate(), B=ScalarParameter() {defaultValue=15, group=AdditonalTechArt_Eye, parameterName=EyeTiling}) {constB=15}
   Time=Multiply(A=Time(), B=ScalarParameter() {defaultValue=0.2, group=AdditonalTechArt_Eye, parameterName=EyeTimeSpeed}) {constB=0.2}
  )
  n10 = NamedRerouteUsage()
  n11 = Multiply(A=Multiply(
    A=TextureSample(UVs=@n9) {samplerType=SAMPLERTYPE_LinearGrayscale, texture=/Game/Global/Textures/Default/Noise/EqualizerComponents_v2_Noise}.RGB
    B=@n10
   )) {constB=0.8}
  n12 = VectorParameter() {group=AdditonalTechArt_Eye, parameterName=EyeMainColor, defaultValue=<rgba 1 0.5185 0 1>}
  NamedRerouteDeclaration(None=LinearInterpolate(A=Add(A=@n8, B=@n11), B=Multiply(A=@n12.RGB) {constB=3}, Alpha=Power(Base=@n10) {constExponent=5})) {name=CachedEyeEmission, nodeColor=<rgba 0 1 0.1094 1>}
  n13 = NamedRerouteUsage()
  n14 = Constant()
  n15 = ScalarParameter() {defaultValue=2, group=AdditonalTechArt, parameterName=NormalIntensityForFresnel}
  n16 = NamedRerouteDeclaration(None=MaterialFunctionCall(Normal=NamedRerouteUsage(), Flatness=Subtract(A=@n14, B=@n15)) {materialFunction=&FlattenNormal}.Result) {name=NormalAdditive, nodeColor=<rgba 0 0.8125 1 1>}
  n17 = Fresnel(
   ExponentIn=ScalarParameter() {defaultValue=5, group=AdditonalTechArt, parameterName=AdditionalFresnelExp}
   BaseReflectFractionIn=ScalarParameter() {defaultValue=0.04, group=AdditonalTechArt, parameterName=AdditionalFresnelBaseReflect}
   Normal=StaticSwitchParameter(True=Transform(None=@n16), False=PixelNormalWS()) {group=AdditonalTechArt, parameterName=UseNormalMap}
  )
  n18 = OneMinus(None=@n17)
  n19 = ScalarParameter() {group=AdditonalTechArt, parameterName=InvertFresnel}
  n20 = NamedRerouteUsage()
  n21 = Multiply(
   A=Multiply(
    A=LinearInterpolate(A=@n17, B=@n18, Alpha=@n19)
    B=StaticComponentMaskParameter(None=@n20) {defaultG=true, group=AdditonalTechArt, parameterName=SmokeChannel}
   )
   B=VectorParameter() {group=AdditonalTechArt, parameterName=AdditionFresnelColor, defaultValue=<rgba 1 1 1 1>}.RGB
  )
  n22 = Multiply(A=Saturate(None=Add(A=@n13, B=@n21)), B=ScalarParameter() {defaultValue=1, group=AdditonalTechArt, parameterName=AdditionalFresneltensity})
  n23 = ComponentMask(None=TextureCoordinate()) {g=true}
  n24 = Multiply(
   A=@n22
   B=SmoothStep(
    Min=ScalarParameter() {defaultValue=0.2, group=AdditonalTechArt, parameterName=YGradientFresnelMin}
    Max=ScalarParameter() {defaultValue=0.8, group=AdditonalTechArt, parameterName=YGradientFresnelMax}
    Value=OneMinus(None=@n23)
   ) {constMax=0.7}
  )
  n25 = ScalarParameter() {group=AdditonalTechArt, parameterName=YGradientFresnel}
  NamedRerouteDeclaration(None=Add(A=NamedRerouteUsage(), B=LinearInterpolate(A=@n22, B=@n24, Alpha=@n25))) {name=CachedEmissiveOutput, nodeColor=<rgba 0.0312 0 1 1>}
  NamedRerouteDeclaration(None=@n1.EmissiveColor) {name=CachedEmissiveInput, nodeColor=<rgba 0.2188 1 0 1>}
  NamedRerouteDeclaration(None=@n1.Normal) {name=CachedNormalInput, nodeColor=<rgba 0.4297 0 1 1>}
  n26 = FunctionInput() {inputName=Input TechArtMask, sortPriority=1}
  n27 = NamedRerouteUsage()
  n28 = NamedRerouteUsage()
  NamedRerouteDeclaration(None=MaterialFunctionCall!(
    TechArtMaskInput=NamedRerouteDeclaration(None=@n26) {name=CachedSkinMask, nodeColor=<rgba 0.5703 0 1 1>}
    EmissiveInput=StaticSwitchParameter(True=@n27, False=@n28) {group=AdditonalTechArt, parameterName=UseEyeDistortEmission}
    FlowSpeed
   ) {materialFunction=/Game/Accessories/FORT_Backpacks/Backpack_F_MED_TalonHime/Materials/MF_SmokeBeneath}.Result) {name=CachedSmoke, nodeColor=<rgba 1 0.2812 0 1>}
  NamedRerouteDeclaration(None=@n6.B) {name=CachedEyeArea, nodeColor=<rgba 1 0.0938 0 1>}
  n29 = Constant3Vector() {constant=<rgba 0 0 1 0>}
  n30 = Constant()
  n31 = ScalarParameter() {defaultValue=1, group=AdditonalTechArt, parameterName=TopNormalIntensityForFresnel}
  n32 = Transform(None=MaterialFunctionCall(Normal=NamedRerouteUsage(), Flatness=Subtract(A=@n30, B=@n31)) {materialFunction=&FlattenNormal}.Result)
  n33 = NamedRerouteDeclaration(None=Power(Base=DotProduct(A=@n29, B=@n32)) {constExponent=1.2}) {name=CachedNormalEdge, nodeColor=<rgba 1 0 0.2344 1>}
  n34 = StaticComponentMaskParameter(None=NamedRerouteUsage()) {defaultR=true, group=SmokeBeneath, parameterName=SmokeBeneathChannel}
  NamedRerouteDeclaration(None=LinearInterpolate(B=Multiply(A=@n33, B=@n34), Alpha=ScalarParameter() {group=AdditonalTechArt, parameterName=NormalEdgeIntensity})) {name=CachedTopGradient, nodeColor=<rgba 1 0.2578 0 1>}
 ]}
