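# Master material for the TalonHime pickaxe, layering animated detail smoke onto the emissive of the TalonHime body effect.
# The 2023 character parent PreFX feeds the shared body function, whose emissive is modified before entering PostFX.
# A panned smoke map on a selectable UV channel distorts a mask that tints the glow, with further masks darkening it and boosting edges.
# The UseDetailSmoke static switch, on by default, can bypass this layer.
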Material /BRCosmetics/Weapons/FORT_Melee/Pickaxe_TalonHime/Materials/M_PickAxe_TalonHime_2023
graph {
 properties {
  bAutomaticallySetUsageInEditor=false
  bUseMaterialAttributes=true
  bUsedWithClothing=true
  bUsedWithMeshDeformer=true
  bUsedWithMorphTargets=true
  bUsedWithNanite=true
  bUsedWithSkeletalMesh=true
  pixelDepthOffsetMode=PDOM_Legacy
  refractionMethod=RM_IndexOfRefraction
  shadingModel=MSM_FromMaterialExpression
  subsurfaceProfile=&SS_HeroSkin_02
  usedShadingModels=Default Lit | Subsurface | Clear Coat | Subsurface Profile | Cloth
 }
 nodes [
  LinearInterpolate()
  n1 = MaterialFunctionCall() {materialFunction=&MF_CharacterParent_2023_PreFX}
  n2 = MaterialFunctionCall!(InputMaterial=@n1.Result, Input TechArtMask=@n1.Result) {materialFunction=/Game/Characters/Player/Female/Medium/Bodies/F_MED_TalonHime/Materials/MF_TalonHimeBody}
  n3 = GetMaterialAttributes(None=@n2.Result) {attributeGetTypes [#1]}
  n4 = ScalarParameter() {defaultValue=0.5, group=DetailSmokeBeneath, parameterName=DetailSmokeColorIntensity}
  n5 = TextureCoordinate()
  n6 = Panner(
   Coordinate=Multiply(A=NamedRerouteUsage(), B=ScalarParameter() {defaultValue=1, group=DetailSmokeBeneath, parameterName=DetailSmoke_DistortTiling})
   Speed=VectorParameter() {group=DetailSmokeBeneath, parameterName=DetailSmokeSpeed, defaultValue=<rgba 0 0.3 0 1>}.RGB
  ) {speedY=0.1}
  n7 = Frac(None=@n6)
  n8 = TextureSampleParameter2D(UVs=@n6) {
   group=SmokeBeneath
   parameterName=Smoke_DistortUV
   samplerSource=SSM_Wrap_WorldGroupSettings
   samplerType=SAMPLERTYPE_LinearColor
   texture=/Game/Accessories/FORT_Backpacks/Backpack_F_MED_TalonHime/Textures/T_Smokemap
  }
  n9 = Multiply(A=ComponentMask(None=@n8.RGB) {g=true, r=true}) {constB=2}
  n10 = Multiply(
   A=SphereMask(A=ComponentMask(None=@n7) {g=true}, B=Constant() {r=0.5}) {attenuationRadius=0.5, hardnessPercent=70}
   B=Normalize(VectorInput=Subtract(A=@n9))
  )
  n11 = ScalarParameter() {defaultValue=0.01, group=DetailSmokeBeneath, parameterName=DetailSmokeDistortIntensity}
  n12 = TextureSample(UVs=Add(A=@n5, B=Multiply(A=@n10, B=@n11))) {
   samplerSource=SSM_Wrap_WorldGroupSettings
   samplerType=SAMPLERTYPE_LinearColor
   texture=/BRCosmetics/Weapons/FORT_Melee/Pickaxe_TalonHime/Textures/T_AdditionalSmoke_Mask
  }
  n13 = Saturate(None=Multiply(A=@n12.RGB) {constB=1.2})
  n14 = LinearInterpolate(
   A=@n3.Emissive Color
   B=Multiply(
    A=Multiply(A=@n3.Emissive Color, B=@n4) {constB=2}
    B=VectorParameter() {group=DetailSmokeBeneath, parameterName=SecondColorTint, defaultValue=<rgba 1 0 0 1>}.RGB
   )
   Alpha=@n13
  )
  n15 = TextureSample(UVs=@n5) {
   samplerSource=SSM_Wrap_WorldGroupSettings
   samplerType=SAMPLERTYPE_LinearColor
   texture=&T_AdditionalSmoke_Mask
  }
  n16 = LinearInterpolate(
   A=@n14
   B=Multiply(A=@n14, B=ScalarParameter() {defaultValue=0.5, group=DetailSmokeBeneath, parameterName=DetailSmokeDarkColorIntensity})
   Alpha=Power(Base=@n15.R, Exp=ScalarParameter() {defaultValue=3, group=DetailSmokeBeneath, parameterName=DetailSmokeDarkPow})
  )
  n17 = LinearInterpolate(
   A=Multiply(A=@n16, B=ScalarParameter() {defaultValue=1.5, group=DetailSmokeBeneath, parameterName=DetailEdgeColorIntensity}) {constB=1.1}
   B=@n16
   Alpha=TextureSample() {texture=/BRCosmetics/Weapons/FORT_Melee/Pickaxe_TalonHime/Textures/T_Pickaxe_TalonHime_Distort}.RGB
  )
  MaterialFunctionCall(
   Material=StaticSwitchParameter(
    True=SetMaterialAttributes(MaterialAttributes=@n3.MaterialAttributes, Emissive Color=@n17) {attributeSetTypes [#1]}
    False=@n2.Result
   ) {defaultValue=true, parameterName=UseDetailSmoke}
   M=@n1.M
   Pre Skinned Local Position=@n1.M
   Distance Blend=@n1.M
   Fresnel=@n1.M
  ) {materialFunction=&MF_CharacterParent_2023_PostFX}
  n18 = ScalarParameter() {group=SmokeBeneath, parameterName=Smoke_UseSecondUVChannelForPanning}
  NamedRerouteDeclaration(None=LinearInterpolate(A=TextureCoordinate(), B=TextureCoordinate() {coordinateIndex=1}, Alpha=Round(None=@n18))) {name=CachedBlendUV, nodeColor=<rgba 0.3359 0 1 1>}
  Power(Base=@n13)
 ]
}
